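# Widget container for a challenge progress bar's stage markers and boss name.
# InitializeStageMarkers spawns markers, spaces them by render translation, activates the first and applies boss-stage textures where the phase target equals 1; PreConstruct previews this at design time.
# UpdateStageMarkerStates completes one marker and activates the next, while Anim_HideStageMarkers fades markers out and the boss name in.
# On Construct, touch mobile devices get a width chosen by viewport aspect ratio.
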WidgetBlueprint /AIProgressBar/UI/WBPs/WBP_Challenge_StageMarkers_Container
animations [Anim_HideStageMarkers {
  binding_count=2
  length_seconds=0.5
  object=&self:Anim_HideStageMarkers
  sequence {bindings [
    Text_BossName {
     id=#1
     possessed_class=CommonTextBlock
     tracks [
      {
       type=MovieSceneFloatTrack
       display_name=Render Opacity
       sections [{type=MovieSceneFloatSection, end_frame=10, start_frame=0, channels [None {type=MovieSceneScriptingFloatChannel, keys [[0,0],[5,0],[10,1]]}]}]
      }
      {
       type=MovieScene2DTransformTrack
       sections [{
         type=MovieScene2DTransformSection
         end_frame=10
         start_frame=5
         channels [
          Scale.X {type=MovieSceneScriptingFloatChannel, keys [[5,1.1],[10,1]]}
          Scale.Y {type=MovieSceneScriptingFloatChannel, keys [[5,1.1],[10,1]]}
         ]
        }]
      }
     ]
    }
    Overlay_BottomMarkers {
     id=#2
     possessed_class=Overlay
     tracks [{
       type=MovieSceneFloatTrack
       display_name=Render Opacity
       sections [{type=MovieSceneFloatSection, end_frame=5, start_frame=0, channels [None {type=MovieSceneScriptingFloatChannel, keys [[0,1],[5,0]]}]}]
      }]
    }
   ]}
 }]
graphs [
 EventGraph {nodes [
   AnimateStageMarkersVisibility() {is=title K2Node_CustomEvent}
   > PlayAnimation(InAnimation=Get Anim_HideStageMarkers() -> (Anim_HideStageMarkers:Widget Animation Object Reference))
   Event Construct()
   > SetWidthMobile()
   n1 = Event PreConstruct()
   > Branch(Condition=@n1.IsDesignTime)
   > InitializeStageMarkers!(
    execute
    self
    ProgressBarUIWidth:double=400
    TotalNumberOfPhases:int=3
    PhaseProgressTargets:Array of Integers
    BossName:Text="LOCTEXT(\"Name\")"
    InitialAnimationDelay:double=0
   ) -> (then)
  ]}
 InitializeStageMarkers {nodes [
   n1 = InitializeStageMarkers!() -> (
    then
    ProgressBarUIWidth:double=~
    TotalNumberOfPhases:int
    PhaseProgressTargets:Array of Integers
    BossName:Text
    InitialAnimationDelay:double=~
   ) {is=title K2Node_FunctionEntry}
   > Clear(TargetArray:Array of WBP Challenge Stage Marker Object References=Get StageMarkersArray() -> (StageMarkersArray:Array of WBP Challenge Stage Marker Object References))
   > ClearChildren(self=Get Overlay_BottomMarkers() -> (Overlay_BottomMarkers:Overlay Object Reference))
   > n2 = For Loop(LastIndex=int - int(A=@n1.TotalNumberOfPhases))
   > n3 = Create Widget!(
    execute:Exec
    Class:User Widget Class Reference=Get ChallengeMarkerClass() -> (ChallengeMarkerClass:WBP Challenge Stage Marker Class Reference)
    OwningPlayer:Player Controller Object Reference
   ) -> (then:Exec, ReturnValue:WBP Challenge Stage Marker Object Reference)
   > Add(
    TargetArray:Array of WBP Challenge Stage Marker Object References=Get StageMarkersArray()
    NewItem:"WBP Challenge Stage Marker Object Reference (by ref)"=@n3
   ) {is=title K2Node_CallArrayFunction}
   > AddChild(self=Get Overlay_BottomMarkers(), Content=@n3)
   n4 = Get TotalNumberOfPhases() -> (TotalNumberOfPhases:int=0)
   n5 = int - int(A=@n4)
   SetStageMarkersPosition!(
    execute=@n2.Completed
    self
    ProgressBarUIWidth:double=Get ProgressBarUIWidth() -> (ProgressBarUIWidth:double)
    ProgressByStages:double=SafeDivide(A=1, B="To Float (Integer)"(InInt=@n5))
    TotalNumberOfPhases:int=@n4
   ) -> (then)
   > SetStageMarkerVisuals!(
    execute
    self
    TotalNumberOfPhases:int=Get TotalNumberOfPhases() -> (TotalNumberOfPhases=0)
    PhaseProgressTargets:Array of Integers=Get PhaseProgressTargets() -> (PhaseProgressTargets:Array of Integers)
    InitialAnimationDelay:double=Get InitialAnimationDelay() -> (InitialAnimationDelay:double)
   ) -> (then)
   > SetBossName!(execute, self, InBossName:Text=Get BossName() -> (BossName:Text)) -> (then)
  ]}
 IsOnBossStage {nodes [
   n1 = IsOnBossStage!() -> (then, PhaseProgressTargets:Array of Integers, PhaseIndex:int) {is=title K2Node_FunctionEntry}
   > Return Node!(
    execute
    IsOnBossStage:bool="Equal (Integer)"(A="Get (a copy)"(Array:Array of Integers=@n1, Dimension 1=@n1.PhaseIndex) -> (Output:int), B=1)
   )
  ]}
 ResetStageMarkers {nodes [
   ResetStageMarkers() {is=title K2Node_FunctionEntry}
   > n1 = For Each Loop(Array:Array of WBP Challenge Stage Marker Object References=Get StageMarkersArray()) -> (Array Element:WBP Challenge Stage Marker Object Reference)
   > SetState!(execute, self:WBP Challenge Stage Marker Object Reference=@n1, Active:bool=false, DelayOnAnimation:double=0) -> (then)
  ]}
 SetBossName {nodes [
   n1 = SetBossName!() -> (then, InBossName:Text) {is=title K2Node_FunctionEntry}
   > "SetText (Text)"(self=Get Text_BossName() -> (Text_BossName:Common Text Object Reference), InText=@n1)
  ]}
 SetStageMarkerVisuals {nodes [
   SetStageMarkerVisuals!() -> (then, TotalNumberOfPhases:int, PhaseProgressTargets:Array of Integers, InitialAnimationDelay:double=~) {is=title K2Node_FunctionEntry}
   > n1 = For Each Loop(Array=Get StageMarkersArray())
   > n2 = Sequence()
   > SetState(
    self=@n1
    Active=Select(Option 0:bool=false, Option 1:bool=true, Index="Equal (Integer)"(A=@n1.Array Index)) -> (ReturnValue:bool)
    DelayOnAnimation=Get InitialAnimationDelay()
   )
   n3 = Get TotalNumberOfPhases() -> (TotalNumberOfPhases=0)
   > SetVisibility(
    self=@n1
    InVisibility=Select(
     Option 0:ESlateVisibility Enum=Collapsed
     Option 1:ESlateVisibility Enum=SelfHitTestInvisible
     Index="integer <= integer"(A=@n1.Array Index, B=@n3)
    ) -> (ReturnValue:ESlateVisibility Enum)
   )
   UpdateWidgetTextures!(
    execute=@n2.then_1
    self:WBP Challenge Stage Marker Object Reference=@n1
    IsBossStage:bool=IsOnBossStage!(PhaseProgressTargets:Array of Integers=Get PhaseProgressTargets(), PhaseIndex:int=@n1.Array Index) -> (IsOnBossStage:bool) {is=title K2Node_CallFunction}
   ) -> (then)
  ]}
 SetStageMarkersPosition {nodes [
   SetStageMarkersPosition!() -> (then, ProgressBarUIWidth:double=~, ProgressByStages:double=~, TotalNumberOfPhases:int) {is=title K2Node_FunctionEntry}
   > n1 = For Each Loop(Array=Get StageMarkersArray())
   > n2 = Branch(Condition="integer <= integer"(A=@n1.Array Index, B=Get TotalNumberOfPhases() -> (TotalNumberOfPhases=0)))
   n3 = Get SizeBox_Root!(self:WBP Challenge Stage Marker Object Reference=@n1) -> (SizeBox_Root:Size Box Object Reference)
   n4 = Get ProgressByStages() -> (ProgressByStages:double)
   n5 = float - float(
    A=float * float(A=Get ProgressBarUIWidth(), B=float * float(A=@n4, B:int=@n1.Array Index)) {is=title K2Node_PromotableOperator}
    B=float / float(A=Get WidthOverride!(self:Size Box Object Reference=@n3) -> (WidthOverride:float), B=2)
   )
   n6 = Get MarkerPadding() -> (MarkerPadding:double)
   > SetRenderTranslation(self=@n1, Translation=MakeVector2D(X=float - float(A=@n5, B=@n6), Y=1))
   Set Visibility(execute=@n2.else, Visibility=Hidden, self=@n1)
  ]}
 SetWidthMobile {nodes [
   SetWidthMobile() {is=title K2Node_FunctionEntry}
   n1 = GetOwningPlayer()
   > Branch(Condition=AND Boolean(A=IsUsingTouch(PlayerController=@n1), B=IsMobileGame()))
   n2 = GetViewportSize!() -> (ReturnValue_X:double, ReturnValue_Y:double)
   n3 = float / float(A=@n2, B=@n2.ReturnValue_Y)
   > SetWidthOverride(
    self=Get SizeBox_MarkersSize() -> (SizeBox_MarkersSize:Size Box Object Reference)
    InWidthOverride=Select(Option 0:float=650, Option 1:float=400, Index="float <= float"(A=@n3, B=1.5) -> (ReturnValue=~)) -> (ReturnValue:float)
   )
  ]}
 UpdateStageMarkerStates {nodes [
   n1 = UpdateStageMarkerStates!() -> (then, NumberOfPhasesCompleted:int) {is=title K2Node_FunctionEntry}
   > OnStageComplete(self:WBP Challenge Stage Marker Object Reference="Get (a copy)"(Array:Array of WBP Challenge Stage Marker Object References=Get StageMarkersArray(), Dimension 1=int - int(A=@n1)) -> (Output:WBP Challenge Stage Marker Object Reference))
   > SetState(self="Get (a copy)"(Array=Get StageMarkersArray(), Dimension 1=@n1), Active=true, DelayOnAnimation=0)
  ]}
]
variables [
 ChallengeMarkerClass {type=Class, default=/AIProgressBar/UI/WBPs/Simple/WBP_Challenge_SimpleStageMarker.WBP_Challenge_SimpleStageMarker_C}
 MarkerPadding {type=number, default=1}
 StageMarkersArray {type="array<WBP_Challenge_StageMarker_C>", default []}
]
widgets {
 info {inheritedWidgetCount=0, parentClass=&UserWidget, rootWidgetClass=&Overlay, widgetCount=4}
 widgets [
  Overlay_Root {widgetClassPath=&Overlay, widget=Overlay_Root}
  SizeBox_MarkersSize {
   widgetClassPath=&SizeBox
   bIsVariable=true
   parent=Overlay_Root
   widget=SizeBox_MarkersSize
   properties {bOverride_MinDesiredHeight=true, bOverride_WidthOverride=true, minDesiredHeight=10, widthOverride=800}
   slot {class=OverlaySlot}
  }
  Overlay_BottomMarkers {widgetClassPath=&Overlay, bIsVariable=true, parent=SizeBox_MarkersSize, widget=Overlay_BottomMarkers, slot {class=SizeBoxSlot}}
  Text_BossName {
   widgetClassPath=&CommonTextBlock
   bIsVariable=true
   parent=Overlay_Root
   widget=Text_BossName
   properties {
    justification=Center
    renderOpacity=0
    text="{GodName}"
    textTransformPolicy=ToUpper
    font {
     bForceMonospaced=false
     bMaterialIsStencil=false
     fontMaterial=null
     fontObject=&BurbankSmall
     letterSpacing=0
     monospacedWidth=0.552
     size=21
     skewAmount=0
     typefaceFontName=Black
     outlineSettings {
      bApplyOutlineToDropShadows=false
      bMiteredCorners=false
      bSeparateFillAlpha=false
      outlineMaterial=null
      outlineSize=0
      outlineColor=<rgba 0.0781 0.0781 0.0781 0.235>
     }
    }
    shadowColorAndOpacity=<rgba 0 0 0 0.2>
    shadowOffset=<1.2 1.2>
   }
   slot {class=OverlaySlot, properties {horizontalAlignment=HAlign_Fill, verticalAlignment=VAlign_Bottom}}
  }
 ]
}
