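# Base Fortnite HUD widget for multi-phase boss challenge progress bars, with no layout and empty hooks for visuals.
# It follows the view target's AI progress bar component to the current challenge manager, binding its events when the boss tag matches.
# Manager progress and per-phase targets become a segmented fill value, completed phases and a boss-stage flag.
# It collapses when spectating, on mobile touch, or without an active matching challenge.
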WidgetBlueprint /AIProgressBar/UI/WBPs/WBP_Challenge_WidgetBase
graphs [
 BeginChallengeInitialSetup {nodes [BeginChallengeInitialSetup() {is=title K2Node_FunctionEntry}]}
 BindToChallengeManagerEvents {nodes [
   BindToChallengeManagerEvents() {is=title K2Node_FunctionEntry}
   > Is Valid(InputObject=Get ChallengeManagerRef() -> (ChallengeManagerRef:Challenge Manager Object Reference))
   > Bind Event to On Challenge Begun(self:Challenge Manager Object Reference=Get ChallengeManagerRef(), Delegate=Create Event()) {is=title K2Node_AddDelegate}
   > Bind Event to On Total Progress Updated(self:Challenge Manager Object Reference=Get ChallengeManagerRef(), Delegate=Create Event()) {is=title K2Node_AddDelegate}
   > Bind Event to On Boss Progress Updated(self:Challenge Manager Base Object Reference=Get ChallengeManagerRef(), Delegate=Create Event()) {is=title K2Node_AddDelegate}
   > Bind Event to On Boss Ready for Final Fight(self:Challenge Manager Object Reference=Get ChallengeManagerRef(), Delegate=Create Event()) {is=title K2Node_AddDelegate}
  ]}
 CalculatePhasesPreviouslyCompleted {nodes [
   CalculatePhasesPreviouslyCompleted() {is=title K2Node_FunctionEntry}
   > n1 = Set SumOfTargetsForCompleteEachPhase(SumOfTargetsForCompleteEachPhase:Array of Integers=GetSumOfTargetsForCompleteEachPhase!() -> (SumToCompletePhasesOutArray:Array of Integers)) -> (Output_Get:Array of Integers)
   > n2 = For Each Loop(Array:Array of Integers=@n1) -> (Array Element:int)
   > Branch(Condition="integer >= integer"(A=Get Total Progress() -> (Total Progress:int=0), B=@n2))
   > UpdatePhaseCompletedVariables!(execute, self) -> (then, IsChallengeCompleted:bool)
  ]}
 CalculateProgressBarFill {nodes [
   CalculateProgressBarFill!() -> (then, WithAnimation:bool=~, CurrentTotalProgressPercentage:double=~) {is=title K2Node_FunctionEntry}
   n1 = Get TotalNumberOfPhases() -> (TotalNumberOfPhases:int=0)
   n2 = SafeDivide(A=1, B="To Float (Integer)"(InInt=int - int(A=@n1)))
   n3 = Get NumberOfPhasesCompleted() -> (NumberOfPhasesCompleted:int=0)
   > Set BarProgress(BarProgress:double=float + float(
     A=Get CurrentTotalProgressPercentage() -> (CurrentTotalProgressPercentage:double)
     B=float * float(A:int=@n3, B=@n2) {is=title K2Node_PromotableOperator}
    )) -> (Output_Get:double)
   > UpdateProgressBarFill!(execute, self, With Animation:bool=Get WithAnimation() -> (WithAnimation:bool)) -> (then)
  ]}
 CalculateTotalProgressPercentage {nodes [
   CalculateTotalProgressPercentage() {is=title K2Node_FunctionEntry}
   > n1 = Branch(Condition=IsLastPhase!(self:Challenge Manager Object Reference=Get ChallengeManagerRef(), PhaseIndex:int=Get CurrentPhase() -> (CurrentPhase:int=0)) -> (Value:bool))
   > Return Node!(execute, TotalProgressPercentage:double=0)
   n2 = Sequence(execute=@n1.else)
   > Set SumToCompletePhase(SumToCompletePhase:int=GetSumOfTargetsForCompleteCurrentPhase!() -> (Sum:int=0)) -> (Output_Get:int=0)
   n3 = "To Float (Integer)"(InInt=GetProgressForCurrentPhase!(SumOfTargetsToCompleteCurrentPhase:int=Get SumToCompletePhase() -> (SumToCompletePhase:int=0)) -> (Targets:int=0))
   n4 = "To Float (Integer)"(InInt=GetNumberOfTargetsForCurrentPhase!() -> (Targets:int=0))
   n5 = int - int(A=Get TotalNumberOfPhases() -> (TotalNumberOfPhases=0))
   Return Node(execute=@n2.then_1, TotalProgressPercentage=SafeDivide(A=SafeDivide(A=@n3, B=@n4), B="To Float (Integer)"(InInt=@n5)))
  ]}
 EventGraph {nodes [
   SetChallengeManager() {is=title K2Node_CustomEvent}
   > n1 = Branch(Condition=NOT Boolean(A=IsActivePlayerSpectating!() -> (bIsSpectating:bool)))
   n2 = Get CachedPawnComponent() -> (CachedPawnComponent:BP AIProgress Bar Pawn Component Object Reference)
   > n3 = Is Valid(InputObject=@n2)
   n4 = Get CurrentChallengeManager!(self:BP AIProgress Bar Pawn Component Object Reference=@n2) -> (CurrentChallengeManager:Challenge Manager Object Reference)
   > n5 = Is Valid(InputObject=@n4)
   UnbindFromChallengeManagerEventsAndClearRef(execute=[@n6.else, @n5.Is Not Valid, @n3.Is Not Valid, @n1.else])
   > SetVisibility(InVisibility=Collapsed)
   n6 = Branch(
    execute=@n5.Is Valid
    Condition="Equal (GameplayTag)"(
     A=Get BossToShow_IDTag() -> (BossToShow_IDTag:Gameplay Tag Structure)
     B=GetBossIDTag!(self:Challenge Manager Base Object Reference=@n4) -> (ReturnValue:Gameplay Tag Structure)
    )
   )
   > n7 = Set ChallengeManagerRef(ChallengeManagerRef:Challenge Manager Object Reference=@n4) -> (Output_Get:Challenge Manager Object Reference)
   > Try Print to Shipping Log!(execute, self, InString:String=Append(A="New challenge manager is ", B=GetDisplayName(Object=@n7))) -> (then)
   > BindToChallengeManagerEvents() {is=title K2Node_CallFunction}
   n8 = Get ChallengeManagerRef()
   > n9 = Branch(Condition="Equal (Enum)"(
     A:EChallengeState Enum=Active
     B:EChallengeState Enum=GetCurrentState!(self:Challenge Manager Base Object Reference=@n8) -> (ReturnValue:EChallengeState Enum=None)
    ))
   > BeginChallengeInitialSetup() {is=title K2Node_CallFunction}
   > UpdateProgressBarState()
   > CheckBossReadyForFight()
   SetVisibility(execute=@n9.else, InVisibility=Collapsed)
   OnLocalPlayerViewTargetChanged() {is=title K2Node_CustomEvent}
   > n10 = Sequence!(execute) -> (then_0, then_1, then_2:Exec)
   n11 = Get CachedPawnComponent()
   > Is Valid(InputObject=@n11)
   > Unbind Event from On Current Challenge Manager Changed(self:BP AIProgress Bar Pawn Component Object Reference=@n11, Delegate=Create Event()) {is=title K2Node_RemoveDelegate}
   > Set CachedPawnComponent(CachedPawnComponent:BP AIProgress Bar Pawn Component Object Reference) -> (Output_Get:BP AIProgress Bar Pawn Component Object Reference)
   > Try Print to Shipping Log(InString=OnViewTargetChange unbinding OnCurrentChallengeManagerChanged : success)
   n12 = GetViewTarget!(self:Controller Object Reference=GetOwningPlayer()) -> (ReturnValue:Actor Object Reference)
   Try Print to Shipping Log(execute=@n10.then_1, InString=Append(A="New view target is ", B=GetDisplayName(Object=@n12)))
   > Is Valid(InputObject=@n12)
   n13 = GetComponentByClass!(
    self:Actor Object Reference=@n12
    ComponentClass:Actor Component Class Reference=/AIProgressBar/Blueprints/Components/BP_AIProgressBarPawnComponent.BP_AIProgressBarPawnComponent_C
   ) -> (ReturnValue:BP AIProgress Bar Pawn Component Object Reference)
   > Is Valid(InputObject=@n13)
   > n14 = Set CachedPawnComponent(CachedPawnComponent=@n13)
   > Bind Event to On Current Challenge Manager Changed(self:BP AIProgress Bar Pawn Component Object Reference=@n14, Delegate=Create Event()) {is=title K2Node_AddDelegate}
   > Try Print to Shipping Log(InString=OnViewTargetChange binding OnCurrentChallengeManagerChanged : success)
   SetChallengeManager(execute=@n10.then_2) {is=title K2Node_CallFunction}
   Event Construct()
   > SetMobileVisibility()
   > Bind Event to On Local Player View Target Changed(
    self:Fort HUDContext Object Reference=Get FortHUDContext!() -> (ReturnValue:Fort HUDContext Object Reference) {is=title K2Node_GetBlueprintContextEx}
    Delegate=Create Event()
   ) {is=title K2Node_AddDelegate}
   > OnLocalPlayerViewTargetChanged() {is=title K2Node_CallFunction}
   ResetProgressBarFillValue!(execute, self, Challenge Progress Bar:WBP Challenge Progress Bar Object Reference) -> (then)
   PlayOutroAnimation() {is=title K2Node_CustomEvent}
   OnChallengeCompletedVisualUpdates() {is=title K2Node_CustomEvent}
   PlayIntroAnimation() {is=title K2Node_CustomEvent}
   CheckBossReadyForFight() {is=title K2Node_CustomEvent}
  ]}
 ForceUpdateProgressBarVisibility {nodes [
   n1 = ForceUpdateProgressBarVisibility!() -> (then, Visible:ESlateVisibility Enum=Visible) {is=title K2Node_FunctionEntry}
   > Try Print to Shipping Log(InString=Append(A="Widget visibility updated : ", B=Enum to String!(Enumerator:ESlateVisibility Enum=@n1) -> (ReturnValue:String)))
  ]}
 GetNumberOfTargetsForCurrentPhase {nodes [
   GetNumberOfTargetsForCurrentPhase() {is=title K2Node_FunctionEntry}
   n1 = int - int(A=Get CurrentPhase() -> (CurrentPhase=0))
   n2 = Get PhaseProgressTargetsArray() -> (PhaseProgressTargetsArray:Array of Integers)
   > n3 = Branch(Condition=Is Valid Index!(TargetArray:Array of Integers=@n2, IndexToTest:int=@n1) -> (ReturnValue:bool))
   > Return Node!(execute, Targets:int="Get (a ref)"(Array=@n2, Dimension 1=@n1))
   Return Node(execute=@n3.else, Targets=0)
  ]}
 GetProgressForCurrentPhase {nodes [
   n1 = GetProgressForCurrentPhase!() -> (then, SumOfTargetsToCompleteCurrentPhase:int) {is=title K2Node_FunctionEntry}
   n2 = GetNumberOfTargetsForCurrentPhase() -> (Targets=0) {is=title K2Node_CallFunction}
   > Return Node(Targets=int - int(A=Get Total Progress() -> (Total Progress=0), B=int - int(A=@n1, B=@n2)))
  ]}
 GetSumOfTargetsForCompleteCurrentPhase {nodes [
   GetSumOfTargetsForCompleteCurrentPhase() {is=title K2Node_FunctionEntry}
   > n1 = For Loop(LastIndex=Get NumberOfPhasesCompleted() -> (NumberOfPhasesCompleted=0))
   n2 = Get PhaseProgressTargetsArray()
   > Set SumToCompletePhase(SumToCompletePhase=int + int(A="Get (a copy)"(Array:Array of Integers=@n2, Dimension 1=@n1) -> (Output:int), B=Get SumToCompletePhase() -> (SumToCompletePhase=0))) -> (Output_Get=0)
   Return Node!(execute=@n1.Completed, Sum:int=Get SumToCompletePhase() -> (SumToCompletePhase=0))
  ]}
 GetSumOfTargetsForCompleteEachPhase {nodes [
   GetSumOfTargetsForCompleteEachPhase() {is=title K2Node_FunctionEntry}
   n1 = Get PhaseProgressTargetsArray()
   > n2 = For Loop(LastIndex=int - int(A=Length(TargetArray:Array of Integers=@n1)))
   n3 = Get PhaseProgressTargetsArray()
   > n4 = Set SumToCompleteCurrentPhase(SumToCompleteCurrentPhase:int=int + int(A="Get (a copy)"(Array=@n3, Dimension 1=@n2), B=Get SumToCompleteCurrentPhase() -> (SumToCompleteCurrentPhase:int=0))) -> (Output_Get:int=0)
   > Add(TargetArray:Array of Integers=Get SumToCompletePhasesArray() -> (SumToCompletePhasesArray:Array of Integers), NewItem:"Integer (by ref)"=@n4) {is=title K2Node_CallArrayFunction}
   Return Node!(execute=@n2.Completed, SumToCompletePhasesOutArray:Array of Integers=Get SumToCompletePhasesArray())
  ]}
 HandleBossProgressUpdatedInManager {nodes [
   HandleBossProgressUpdatedInManager() {is=title K2Node_FunctionEntry}
   > HandleBossProgressUpdatedInternal!(execute, self, AnimateProgressBar:bool=true) -> (then)
  ]}
 HandleBossProgressUpdatedInternal {nodes [
   n1 = HandleBossProgressUpdatedInternal!() -> (then, AnimateProgressBar:bool=~) {is=title K2Node_FunctionEntry}
   > UpdateLocalBossProgressWithManagerValue()
   > UpdateProgressBarFill(With Animation=@n1)
  ]}
 HandleTotalProgressUpdatedInManager {nodes [
   HandleTotalProgressUpdatedInManager() {is=title K2Node_FunctionEntry}
   > UpdateLocalTotalProgressWithManagerValue()
   > HandleTotalProgressUpdatedInternal!(execute, self, AnimateVisuals:bool=true) -> (then)
  ]}
 HandleTotalProgressUpdatedInternal {nodes [
   HandleTotalProgressUpdatedInternal!() -> (then, AnimateVisuals:bool=~) {is=title K2Node_FunctionEntry}
   > n1 = Sequence()
   > n2 = CalculateTotalProgressPercentage!(execute, self) -> (then, TotalProgressPercentage:double) {is=title K2Node_CallFunction}
   > CalculateProgressBarFill!(execute, self, WithAnimation:bool=Get AnimateVisuals() -> (AnimateVisuals:bool), CurrentTotalProgressPercentage:double=@n2) -> (then) {is=title K2Node_CallFunction}
   n3 = IsCurrentPhaseCompleted!(execute=@n1.then_1, self) -> (then, PhaseCompleted:bool)
   > Branch(Condition=@n3)
   > n4 = UpdatePhaseCompletedVariables()
   > Branch(Condition=Get AnimateVisuals())
   > UpdateChallengeCompleted!(execute, self, IsChallengeCompleted:bool=@n4) -> (then)
  ]}
 InitialSetupOfChallengeVariables {nodes [
   InitialSetupOfChallengeVariables() {is=title K2Node_FunctionEntry}
   > Is Valid(InputObject=Get ChallengeManagerRef())
   > Set ChallengeUIData(ChallengeUIData:UIChallenge Data Structure) -> (Output_Get:UIChallenge Data Structure)
   > Set TotalNumberOfTargets(TotalNumberOfTargets:int=0) -> (Output_Get:int=0)
   > Set TotalNumberOfPhases(TotalNumberOfPhases:int=0) -> (Output_Get:int=0)
   > Set NumberOfPhasesCompleted(NumberOfPhasesCompleted:int=0) -> (Output_Get:int=0)
   > Set CurrentPhase(CurrentPhase:int=1) -> (Output_Get:int=0)
   > Clear(TargetArray:Array of Integers=Get PhaseProgressTargetsArray())
   > Set PhaseProgressTargetsArray(PhaseProgressTargetsArray:Array of Integers=Get PhaseProgressTargets!(self:Challenge Manager Base Object Reference=Get ChallengeManagerRef()) -> (PhaseProgressTargets:Array of Integers)) -> (Output_Get:Array of Integers)
   > n1 = For Each Loop(Array=Get PhaseProgressTargetsArray())
   > n2 = Branch(Condition="Equal (Integer)"(A=@n1))
   > Remove Index!(execute, TargetArray:Array of Integers=Get PhaseProgressTargetsArray(), IndexToRemove:int=@n1.Array Index) -> (then)
   Set TotalNumberOfTargets(execute=@n2.else, TotalNumberOfTargets=int + int(A=Get TotalNumberOfTargets() -> (TotalNumberOfTargets:int=0), B=@n1)) -> (Output_Get=0)
   Set TotalNumberOfPhases(execute=@n1.Completed, TotalNumberOfPhases=Length(TargetArray=Get PhaseProgressTargetsArray())) -> (Output_Get=0)
  ]}
 InitialSetupOfTextsAndStageMarkers {nodes [
   InitialSetupOfTextsAndStageMarkers!() -> (
    then
    Challenge_StageIndicatorText:WBP Challenge Stage Indicator Text Object Reference
    Challenge Stage Markers Container:WBP Challenge Stage Markers Container Object Reference
    Challenge Progress Bar:WBP Challenge Progress Bar Object Reference
   ) {is=title K2Node_FunctionEntry}
   n1 = Get ChallengeManagerRef()
   > Is Valid(InputObject=@n1)
   > n2 = Set ChallengeUIData!(execute, ChallengeUIData=Get UIChallengeData!(self:Challenge Manager Object Reference=@n1) -> (UIChallengeData:UIChallenge Data Structure)) -> (
    then
    Output_Get_DefeatTheBossText_2_#1:Text
    Output_Get_BossNameText_14_#2:Text
    Output_Get_EliminateTheMinionsText_11_#3:Text
    Output_Get_BossIconTexture_10_#4:Texture Object Reference
    Output_Get_MinionIconTexture_18_#5:Texture Object Reference
   )
   n3 = Get Challenge_StageIndicatorText() -> (Challenge_StageIndicatorText:WBP Challenge Stage Indicator Text Object Reference)
   > n4 = Is Valid(InputObject=@n3)
   > Setup Stage Information!(
    execute
    self:WBP Challenge Stage Indicator Text Object Reference=@n3
    BossStageText:Text=@n2
    MinionsStageText:Text=@n2.Output_Get_EliminateTheMinionsText_11_#3
    BossTexture:Texture Object Reference=@n2.Output_Get_BossIconTexture_10_#4
    MinionTexture:Texture Object Reference=@n2.Output_Get_MinionIconTexture_18_#5
   ) -> (then)
   > n5 = UpdateStageText!(
    execute
    self:WBP Challenge Stage Indicator Text Object Reference=@n3
    IsBossStage:bool=IsOnBossStage!() -> (IsOnBossStage:bool)
    AnimationDelay:double=Get InitialAnimationDelay() -> (InitialAnimationDelay:double)
    WithDelayedOutroAnim:bool=true
   ) -> (then)
   n6 = Get Challenge Stage Markers Container() -> (Challenge Stage Markers Container:WBP Challenge Stage Markers Container Object Reference)
   n7 = Is Valid(exec=[@n5.then, @n4.Is Not Valid], InputObject=@n6)
   n8 = Get!() -> (
    ChallengeUIData_DefeatTheBossText_2_#1:Text
    ChallengeUIData_BossNameText_14_#2:Text
    ChallengeUIData_EliminateTheMinionsText_11_#3:Text
    ChallengeUIData_BossIconTexture_10_#4:Texture Object Reference
    ChallengeUIData_MinionIconTexture_18_#5:Texture Object Reference
   )
   > n9 = InitializeStageMarkers!(
    execute
    self:WBP Challenge Stage Markers Container Object Reference=@n6
    ProgressBarUIWidth:double=GetSizeBoxWidth!(self:WBP Challenge Progress Bar Object Reference=Get Challenge Progress Bar() -> (Challenge Progress Bar:WBP Challenge Progress Bar Object Reference)) -> (Width:double)
    TotalNumberOfPhases:int=Get TotalNumberOfPhases() -> (TotalNumberOfPhases=0)
    PhaseProgressTargets:Array of Integers=Get PhaseProgressTargetsArray()
    BossName:Text=@n8.ChallengeUIData_BossNameText_14_#2
    InitialAnimationDelay:double=Get InitialAnimationDelay()
   ) -> (then)
   n10 = Get Challenge Progress Bar()
   Is Valid(exec=[@n9.then, @n7.Is Not Valid], InputObject=@n10)
   > SetupBossIcon!(
    execute
    self:WBP Challenge Progress Bar Object Reference=@n10
    BossIcon:Texture Object Reference=@n8.ChallengeUIData_BossIconTexture_10_#4
   ) -> (then)
   > PlayShakeAnimation!(execute, self:WBP Challenge Progress Bar Object Reference=@n10, AnimationDelay:double=Get InitialAnimationDelay()) -> (then)
  ]}
 IsActivePlayerSpectating {nodes [
   IsActivePlayerSpectating() {is=title K2Node_FunctionEntry}
   > n1 = Cast To FortPlayerControllerAthena(Object=GetOwningPlayer())
   n2 = GetIsPlayingReplay!() -> (ReturnValue:bool)
   > n3 = Branch(Condition=AND Boolean(A=GetIsSpectating!(self:Fort Player Controller Athena Object Reference=@n1) -> (ReturnValue:bool), B=NOT Boolean(A=@n2)))
   > Return Node!(execute, bIsSpectating:bool=true)
   Return Node(execute=@n3.else, bIsSpectating=false)
  ]}
 IsCurrentPhaseCompleted {nodes [
   IsCurrentPhaseCompleted() {is=title K2Node_FunctionEntry}
   n1 = GetSumOfTargetsForCompleteCurrentPhase() -> (Sum=0) {is=title K2Node_CallFunction}
   n2 = Get Total Progress() -> (Total Progress=0)
   > Return Node!(execute, PhaseCompleted:bool="Equal (Integer)"(A=int - int(A=@n2, B=@n1)))
  ]}
 IsOnBossStage {nodes [
   IsOnBossStage() {is=title K2Node_FunctionEntry}
   n1 = int - int(A=Get CurrentPhase() -> (CurrentPhase=0))
   n2 = Get PhaseProgressTargetsArray()
   > Return Node!(execute, IsOnBossStage:bool="Equal (Integer)"(A="Get (a copy)"(Array=@n2, Dimension 1=@n1), B=1))
  ]}
 OnBossReadyForFinalFight {nodes [
   OnBossReadyForFinalFight() {is=title K2Node_FunctionEntry}
   > UpdateChallengeInformation!(execute, self, WithAnimation:bool=true) -> (then)
   > Try Print to Shipping Log(InString=Boss ready for final fight)
  ]}
 PhaseCompletedVisualUpdates {nodes [PhaseCompletedVisualUpdates() {is=title K2Node_FunctionEntry}]}
 ResetProgressBarFillValue {nodes [
   n1 = ResetProgressBarFillValue!() -> (then, Challenge Progress Bar:WBP Challenge Progress Bar Object Reference) {is=title K2Node_FunctionEntry}
   > SetProgressDirectly!(execute, self:WBP Challenge Progress Bar Object Reference=@n1, Progress:double=0) -> (then)
  ]}
 SetMobileVisibility {nodes [
   SetMobileVisibility() {is=title K2Node_FunctionEntry}
   n1 = GetOwningPlayer()
   > Branch(Condition=AND Boolean(A=IsUsingTouch(PlayerController=@n1), B=IsMobileGame()))
   > SetVisibility(InVisibility=Collapsed)
  ]}
 Try Print to Shipping Log {nodes [
   n1 = Try Print to Shipping Log!() -> (then, InString:String) {is=title K2Node_FunctionEntry}
   > Branch(Condition=Get EnableDebugShippingLogs() -> (EnableDebugShippingLogs:bool))
   > FortShippingLog!(execute, WorldContextObject:Object Reference=Self-Reference(), InString:String=@n1, bLogAsWarning:bool=false) -> (then)
  ]}
 UnbindFromChallengeManagerEventsAndClearRef {nodes [
   UnbindFromChallengeManagerEventsAndClearRef() {is=title K2Node_FunctionEntry}
   > Is Valid(InputObject=Get ChallengeManagerRef())
   > Unbind Event from On Challenge Begun(self:Challenge Manager Object Reference=Get ChallengeManagerRef(), Delegate=Create Event()) {is=title K2Node_RemoveDelegate}
   > Unbind Event from On Total Progress Updated(self:Challenge Manager Object Reference=Get ChallengeManagerRef(), Delegate=Create Event()) {is=title K2Node_RemoveDelegate}
   > Unbind Event from On Boss Progress Updated(self:Challenge Manager Object Reference=Get ChallengeManagerRef(), Delegate=Create Event()) {is=title K2Node_RemoveDelegate}
   > Unbind Event from On Boss Ready for Final Fight(self:Challenge Manager Object Reference=Get ChallengeManagerRef(), Delegate=Create Event()) {is=title K2Node_RemoveDelegate}
   > Set ChallengeManagerRef()
   > Try Print to Shipping Log(InString=Cleared challenge manager) {is=title K2Node_CallFunction}
  ]}
 UpdateChallengeCompleted {nodes [
   n1 = UpdateChallengeCompleted!() -> (then, IsChallengeCompleted:bool=~) {is=title K2Node_FunctionEntry}
   > n2 = Branch(Condition=@n1)
   > Reroute Node()
   Reroute Node(InputPin=@n2.else)
   OnChallengeCompletedVisualUpdates() {is=title K2Node_CallFunction}
   Reroute Node(InputPin:Wildcard) -> (OutputPin:Wildcard)
   Reroute Node()
   Reroute Node()
  ]}
 UpdateChallengeInformation {nodes [
   Set Progress Bar Style!(
    execute
    self:Object Reference=Get WBP_GodChallenges_ProgressBar() -> (WBP_GodChallenges_ProgressBar:Object Reference)
    HasReachedBossStage:bool=IsOnBossStage() {is=title K2Node_CallFunction}
   ) -> (then) {is=title K2Node_CallFunction}
   > PlayAnimation(InAnimation=Get Anim_Intro() -> (Anim_Intro:Widget Animation Object Reference), PlaybackSpeed=2)
   > n1 = Branch(Condition=Get WithAnimation())
   Animate Stage Markers Visibility(
    execute=[@n1.else, @n1.then]
    self:Object Reference=Get WBP_GodChallenges_StageMarkers() -> (WBP_GodChallenges_StageMarkers:Object Reference)
   )
   UpdateChallengeInformation!() -> (then, WithAnimation:bool=~) {is=title K2Node_FunctionEntry}
  ]}
 UpdateLocalBossProgressWithManagerValue {nodes [
   UpdateLocalBossProgressWithManagerValue() {is=title K2Node_FunctionEntry}
   n1 = Get ChallengeManagerRef()
   > Is Valid(InputObject=@n1)
   > Set Boss Progress(Boss Progress:double=Get BossProgress!(self:Challenge Manager Base Object Reference=@n1) -> (BossProgress:double)) -> (Output_Get:double)
  ]}
 UpdateLocalTotalProgressWithManagerValue {nodes [
   UpdateLocalTotalProgressWithManagerValue() {is=title K2Node_FunctionEntry}
   n1 = Get ChallengeManagerRef()
   > Is Valid(InputObject=@n1)
   > Set Total Progress(Total Progress:int=Get TotalProgress!(self:Challenge Manager Base Object Reference=@n1) -> (TotalProgress:int=0)) -> (Output_Get:int=0)
  ]}
 UpdatePhaseCompletedVariables {nodes [
   UpdatePhaseCompletedVariables() {is=title K2Node_FunctionEntry}
   > Increment Int(Value=Get CurrentPhase() -> (CurrentPhase=0))
   > Increment Int(Value=Get NumberOfPhasesCompleted() -> (NumberOfPhasesCompleted=0))
   > Return Node!(
    execute
    IsChallengeCompleted:bool="Equal (Integer)"(A=Get NumberOfPhasesCompleted() -> (NumberOfPhasesCompleted=0), B=Get TotalNumberOfPhases() -> (TotalNumberOfPhases=0))
   )
  ]}
 UpdateProgressBarFill {nodes [UpdateProgressBarFill!() -> (then, With Animation:bool=~) {is=title K2Node_FunctionEntry}]}
 UpdateProgressBarState {nodes [
   UpdateProgressBarState() {is=title K2Node_FunctionEntry}
   > UpdateLocalTotalProgressWithManagerValue() {is=title K2Node_CallFunction}
   > CalculatePhasesPreviouslyCompleted() {is=title K2Node_CallFunction}
   > HandleTotalProgressUpdatedInternal(AnimateVisuals=false) {is=title K2Node_CallFunction}
  ]}
 UpdateStageMarkersState {nodes [
   UpdateStageMarkersState() {is=title K2Node_FunctionEntry}
   Get NumberOfPhasesCompleted() -> (NumberOfPhasesCompleted=0)
  ]}
]
variables [
 Boss Progress {type=number}
 BossReadyForFinalFight {type=boolean}
 BossToShow_IDTag {type=GameplayTag, default {tagName=None}}
 CachedPawnComponent {type=BP_AIProgressBarPawnComponent_C, default=null}
 ChallengeManagerRef {type=ChallengeManager_C, default=null}
 ChallengeStageText {type="array<string>", default []}
 ChallengeUIData {
  type=UIChallengeData
  default {bossIconTexture=null, bossNameText="", defeatTheBossText="", eliminateTheMinionsText="", minionIconTexture=null}
 }
 CurrentPhase {type=integer, default=1}
 EnableDebugShippingLogs {type=boolean}
 InitialAnimationDelay {type=number, default=0.5}
 NumberOfPhasesCompleted {type=integer}
 PhaseProgressTargetsArray {type="array<integer>"}
 StageMarkerSize {type=number, default=50}
 TimeToHideWidgetOnceFinished {type=number, default=1}
 Total Progress {type=integer}
 TotalNumberOfPhases {type=integer}
 TotalNumberOfTargets {type=integer}
]
widgets {info {inheritedWidgetCount=0, parentClass=&FortHUDElementWidget, rootWidgetClass=null, widgetCount=0}, widgets []}
